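# Physics asset for the Junk Samurai hat/face accessory, providing collision and joints for the dynamic helmet pieces.
# Six box bodies form three two-bone chains (dyn_helmet L, R and B) hanging from the head, with capsules on head, spine_05 and clavicles approximating the upper body.
# The setup is asymmetric: the left chain and B tip joint have free swing and twist, unlike the right chain and B root joint.
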PhysicsAsset /BRCosmetics/Accessories/Hats/M_MED_JunkSamurai_FaceAcc/Meshes/Junk_Samurai_Physics
bodies {
 clavicle_l [{shapeType=Capsule, length=0, radius=9.755, center=<0.4941 -1.8088 1.6537>, rotation=<rot 0 -90 0>}]
 clavicle_r [{shapeType=Capsule, length=0, radius=9.755, center=<0 0 -1.0755>, rotation=<rot 0 -90 0>}]
 dyn_helmet_B_01 [{shapeType=Box, extentX=3, extentY=16.9915, extentZ=2, center=<1.5384 0 0.2401>}]
 dyn_helmet_B_02 [{shapeType=Box, extentX=8.854, extentY=20.2192, extentZ=3.5, center=<4.8321 -0.0001 0.1733>}]
 dyn_helmet_L_01 [{shapeType=Box, extentX=4.6308, extentY=16.8782, extentZ=2, center=<1.7286 0.0001 0.0709>, rotation=<rot -2.046 0 0>}]
 dyn_helmet_L_02 [{shapeType=Box, extentX=7.1308, extentY=18.3093, extentZ=1.7089, center=<3.9109 -0.3697 -0.007>, rotation=<rot -2.046 0 0>}]
 dyn_helmet_R_01 [{shapeType=Box, extentX=4.3277, extentY=17.2835, extentZ=2, center=<2.1967 0 0.0001>}]
 dyn_helmet_R_02 [{shapeType=Box, extentX=7.8277, extentY=16.2412, extentZ=1.5, center=<3.8787 0.433 0.0001>}]
 head [
  {shapeType=Capsule, length=4.0687, radius=7.4304, center=<-7.361 -2.6325 0.0002>}
  {shapeType=Capsule, length=3.5687, radius=8.4304, center=<-3.8867 3.4813 0.0001>}
  {shapeType=Capsule, length=1.0687, radius=7.6804, center=<-1.8232 -0.1631 0>}
 ]
 spine_05 [
  {shapeType=Capsule, length=16.755, radius=10.755, rotation=<rot 0 -90 0>}
  {shapeType=Capsule, length=2.25, radius=7.005, center=<7.7426 -3.2857 6.4582>, rotation=<rot 0 -90 0>}
  {shapeType=Capsule, length=2, radius=7.255, center=<5.9914 -2.6843 -7.7472>, rotation=<rot 0 -90 0>}
 ]
}
constraints [
 {
  bone1Name=dyn_helmet_L_01
  bone2Name=head
  swing1LimitDegrees=5
  swing1Motion=Free
  swing2LimitDegrees=5
  swing2Motion=Free
  twistLimitDegrees=5
  twistMotion=Free
 }
 {
  bone1Name=dyn_helmet_L_02
  bone2Name=dyn_helmet_L_01
  swing1LimitDegrees=5
  swing1Motion=Free
  swing2LimitDegrees=5
  swing2Motion=Free
  twistLimitDegrees=5
  twistMotion=Free
 }
 {bone1Name=dyn_helmet_R_01, bone2Name=head, swing1LimitDegrees=3, swing2LimitDegrees=3, twistLimitDegrees=3}
 {bone1Name=dyn_helmet_R_02, bone2Name=dyn_helmet_R_01, swing1LimitDegrees=3, swing2LimitDegrees=3, twistLimitDegrees=3}
 {bone1Name=dyn_helmet_B_01, bone2Name=head, swing1LimitDegrees=0, swing2LimitDegrees=90, twistLimitDegrees=0}
 {
  bone1Name=dyn_helmet_B_02
  bone2Name=dyn_helmet_B_01
  swing1LimitDegrees=90
  swing1Motion=Free
  swing2LimitDegrees=90
  swing2Motion=Free
  twistLimitDegrees=1
  twistMotion=Free
 }
 {
  bone1Name=head
  bone2Name=spine_05
  swing1LimitDegrees=45
  swing1Motion=Free
  swing2LimitDegrees=45
  swing2Motion=Free
  twistLimitDegrees=45
  twistMotion=Free
 }
 {
  bone1Name=clavicle_r
  bone2Name=spine_05
  swing1LimitDegrees=45
  swing1Motion=Free
  swing2LimitDegrees=45
  swing2Motion=Free
  twistLimitDegrees=45
  twistMotion=Free
 }
 {
  bone1Name=clavicle_l
  bone2Name=spine_05
  swing1LimitDegrees=45
  swing1Motion=Free
  swing2LimitDegrees=45
  swing2Motion=Free
  twistLimitDegrees=45
  twistMotion=Free
 }
]
