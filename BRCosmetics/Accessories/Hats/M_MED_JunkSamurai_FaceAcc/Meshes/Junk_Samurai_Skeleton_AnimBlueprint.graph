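# Post-process style Animation Blueprint for the Junk Samurai hat accessory, adding secondary motion and pose correctives to its helmet and can lid.
# It copies the attached parent's pose, then applies AnimDynamics to the lid and head-space RigidBody physics, weakened in the front end.
# A constraint, additive bone offsets and a spine-driven controller then adjust helmet pieces, several weighted by Helmet_LookUp, computed each update from head rotation relative to neck_02.
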AnimBlueprint /BRCosmetics/Accessories/Hats/M_MED_JunkSamurai_FaceAcc/Meshes/Junk_Samurai_Skeleton_AnimBlueprint
target_skeleton=&/Junk_Samurai_Skeleton
graphs [
 AnimGraph {nodes [
   Anim Dynamics - Bone: dyn_canlid(Alpha=0.6, ExternalForce="0,0,0") {
    angularDampingOverride=0.99
    angularSpringConstant=2
    bGravityOverrideInSimSpace=true
    bOverrideAngularDamping=true
    bOverrideLinearDamping=true
    gravityScale=0.6
    linearDampingOverride=0.99
    alphaBoolBlend {blendOption=Linear}
    boundBone {boneName=dyn_canlid}
    componentLinearAccScale=<0.5 0.5 0.5>
    componentLinearVelScale=<1 1 1>
    physicsBodyDefinitions [{
      sphereCollisionRadius=0
      boundBone {boneName=dyn_canlid}
      boxExtents=<0.2 0.2 0.2>
      constraintSetup {twistAxis=AxisY, angularLimitsMax=<5 95 35.684>, angularLimitsMin=<-5 85 -23.1236>, angularTarget=<0 0 0>}
      localJointOffset=<0 0 -0.1>
     }]
   }
   Copy Pose From Mesh() {bUseAttachedParent=true}
   n1 = "Transform (Modify) Bone - Bone: dyn_helmet_B_Offset"(Alpha=1, Translation="2,-1,2", Rotation="-22,15,0", Scale="1,1,1") {
    rotationMode=BMM_Additive
    scaleMode=BMM_Additive
    translationMode=BMM_Additive
    alphaBoolBlend {blendOption=Linear}
    boneToModify {boneName=dyn_helmet_B_Offset}
    rotation {pitch=-22, yaw=15}
    translation=<2 -1 2>
   }
   "Transform (Modify) Bone - Bone: dyn_helmet_L_Offset"(ComponentPose=@n1, Alpha=1, Translation="-1,3,0", Rotation="-4,20,-2", Scale="1,1,1") {
    rotationMode=BMM_Additive
    scaleMode=BMM_Additive
    translationMode=BMM_Additive
    alphaBoolBlend {blendOption=Linear}
    boneToModify {boneName=dyn_helmet_L_Offset}
    rotation=<rot -4 20 -2>
    translation=<-1 3>
   }
   n2 = Anim Dynamics - Bone: dyn_canlid(ComponentPose=Local To Component(LocalPose=Use cached pose 'AnimCache'()), Alpha=0.05, ExternalForce="0,0,0") {
    angularDampingOverride=0.5
    angularSpringConstant=400
    bAngularSpring=true
    bGravityOverrideInSimSpace=true
    bOverrideAngularDamping=true
    bOverrideLinearDamping=true
    gravityScale=1.3
    linearDampingOverride=0.99
    numSolverIterationsPostUpdate=4
    numSolverIterationsPreUpdate=16
    alphaBoolBlend {blendOption=Linear}
    boundBone {boneName=dyn_canlid}
    physicsBodyDefinitions [{
      collisionType=CustomSphere
      sphereCollisionRadius=1.1
      boundBone {boneName=dyn_canlid}
      boxExtents=<3 3 3>
      constraintSetup {angularTargetAxis=AxisZ, twistAxis=AxisY, angularLimitsMax=<10 10 40>, angularLimitsMin=<-10 -10 -40>, angularTarget=<0 0 0>}
      localJointOffset=<-2 0 0>
     }]
   }
   n3 = Get bIsInFrontEnd()
   n4 = Get bIsPlayingFullBodySlotInFrontEnd() -> (bIsPlayingFullBodySlotInFrontEnd:bool)
   n5 = RigidBody(ComponentPose=@n2, Alpha=SelectFloat(A=0.5, B=1, bPickA=OR Boolean(A=@n3, B=@n4))) {
    bClampLinearTranslationLimitToRefPose=true
    bOverrideWorldGravity=true
    overridePhysicsAsset=&/Junk_Samurai_Physics
    simulationSpace=BaseBoneSpace
    alphaBoolBlend {blendOption=Linear}
    baseBoneRef {boneName=head}
    componentAppliedLinearAccClamp=<1000 1000 1000>
    componentLinearAccScale=<0.5 0.5 0.5>
    componentLinearVelScale=<0.5 0.5 0.5>
    overrideWorldGravity {z=-980}
   }
   n6 = "Constraint - dyn_helmet_B_02 (R)"(ComponentPose=@n5, Alpha=0.5, ConstraintWeights_0=1) {
    alphaBoolBlend {blendOption=Linear}
    boneToModify {boneName=dyn_helmet_B_02}
    constraintSetup [{transformType=Rotation, targetBone {boneName=dyn_helmet_B_01}}]
    constraintWeights [1]
   }
   n7 = Get Helmet_LookUp() -> (Helmet_LookUp:double)
   n8 = "Transform (Modify) Bone - Bone: dyn_helmet_B_Offset"(ComponentPose=@n6, Alpha=@n7, Translation="2,-1,2", Rotation="-22,15,0", Scale="1,1,1") {
    rotationMode=BMM_Additive
    scaleMode=BMM_Additive
    translationMode=BMM_Additive
    alphaBoolBlend {blendOption=Linear}
    boneToModify {boneName=dyn_helmet_B_Offset}
    rotation {pitch=-22, yaw=15}
    translation=<2 -1 2>
   }
   n9 = "Transform (Modify) Bone - Bone: dyn_helmet_L_Offset"(ComponentPose=@n8, Alpha=@n7, Translation="-1,3,0", Rotation="-4,20,-2", Scale="1,1,1") {
    rotationMode=BMM_Additive
    scaleMode=BMM_Additive
    translationMode=BMM_Additive
    alphaBoolBlend {blendOption=Linear}
    boneToModify {boneName=dyn_helmet_L_Offset}
    rotation=<rot -4 20 -2>
    translation=<-1 3>
   }
   n10 = "Transform (Modify) Bone - Bone: dyn_helmet_L_02"(ComponentPose=@n9, Alpha="Clamp (Float)"(Value=@n7, Min=0, Max=0.35), Translation="0,0,0", Rotation="20,10,0", Scale="1,1,1") {
    rotationMode=BMM_Additive
    scaleMode=BMM_Additive
    translationMode=BMM_Additive
    alphaBoolBlend {blendOption=Linear}
    boneToModify {boneName=dyn_helmet_L_02}
    rotation {pitch=20, yaw=10}
   }
   n11 = "Transform (Modify) Bone - Bone: dyn_helmet_R_02"(ComponentPose=@n10, Alpha="Clamp (Float)"(Value=@n7, Min=0, Max=0.2), Translation="-0.1,0,0", Rotation="-40,-10,20", Scale="1,1,1") {
    rotationMode=BMM_Additive
    scaleMode=BMM_Additive
    translationMode=BMM_Additive
    alphaBoolBlend {blendOption=Linear}
    boneToModify {boneName=dyn_helmet_R_02}
    rotation=<rot -40 -10 20>
    translation {x=-0.1}
   }
   n12 = "Transform (Modify) Bone - Bone: dyn_helmet_R_01"(ComponentPose=@n11, Alpha=1, Translation="-0.1,0,0", Rotation="-2.5,-3,3", Scale="1,1,1") {
    rotationMode=BMM_Additive
    scaleMode=BMM_Additive
    translationMode=BMM_Additive
    alphaBoolBlend {blendOption=Linear}
    boneToModify {boneName=dyn_helmet_R_01}
    rotation=<rot -2.5 -3 3>
    translation {x=-0.1}
   }
   n13 = "Transform (Modify) Bone - Bone: dyn_helmet_R_02"(ComponentPose=@n12, Alpha=1, Translation="0.3,0,0", Rotation="3,15,3", Scale="1,1,1") {
    rotationMode=BMM_Additive
    scaleMode=BMM_Additive
    translationMode=BMM_Additive
    alphaBoolBlend {blendOption=Linear}
    boneToModify {boneName=dyn_helmet_R_02}
    rotation=<rot 3 15 3>
    translation {x=0.3}
   }
   n14 = "Transform (Modify) Bone - Bone: dyn_helmet_L_02"(ComponentPose=@n13, Alpha=1, Translation="0,0,0", Rotation="0,-2.5,0", Scale="1,1,1") {
    rotationMode=BMM_Additive
    scaleMode=BMM_Additive
    translationMode=BMM_Additive
    alphaBoolBlend {blendOption=Linear}
    boneToModify {boneName=dyn_helmet_L_02}
    rotation {yaw=-2.5}
   }
   n15 = "dyn_helmet_B_Offset.multiple = spine_05.rotateX * -1 - Bone Driven Controller"!(ComponentPose:Component Space Pose Link Structure=@n14, Alpha:float=0.35) -> (Pose:Component Space Pose Link Structure) {
    is=title AnimGraphNode_BoneDrivenController
    bAffectTargetRotationX=true
    bAffectTargetRotationY=true
    multiplier=-1
    sourceComponent=RotationX
    alphaBoolBlend {blendOption=Linear}
    sourceBone {boneName=spine_05}
    targetBone {boneName=dyn_helmet_B_Offset}
   }
   n16 = "Transform (Modify) Bone - Bone: dyn_helmet_L_Offset"(
    ComponentPose=@n15
    Alpha=float - float(B=Get Helmet_LookUp()) {is=title K2Node_CallFunction}
    Translation="0.5,0,0"
    Rotation="7,0,-5"
    Scale="1,1,1"
   ) {
    rotationMode=BMM_Additive
    scaleMode=BMM_Additive
    translationMode=BMM_Additive
    alphaBoolBlend {blendOption=Linear}
    boneToModify {boneName=dyn_helmet_L_Offset}
    rotation {pitch=7, roll=-5}
    translation {x=0.5}
   }
   Output Pose(Result=Component To Local(ComponentPose=@n16)) {layerGroup=None}
   Save cached pose 'AnimCache'(Pose=Copy Pose From Mesh() {bUseAttachedParent=true}) {cacheName=AnimCache}
  ]}
 EventGraph {nodes [
   Event BlueprintUpdateAnimation()
   n1 = BreakTransform(InTransform=GetDeltaTransformFromRefPose(self=GetOwnerSkeletalMeshComponent(), BoneName=head, BaseName=neck_02))
   > Set Helmet_LookUp(Helmet_LookUp:double=MapRangeClamped(Value=BreakRotator(InRot=@n1.Rotation).Yaw, InRangeA=4, InRangeB=-30)) -> (Output_Get:double)
   TryGetPawnOwner()
  ]}
]
variables [
 Helmet_LookDown {type=number}
 Helmet_LookLeft {type=number}
 Helmet_LookUp {type=number}
]
